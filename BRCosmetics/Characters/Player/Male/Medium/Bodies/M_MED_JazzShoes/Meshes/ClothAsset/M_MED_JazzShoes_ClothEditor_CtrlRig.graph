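# Control Rig for the M_MED_JazzShoes male character's cloth asset, posing apron bone chains around the legs and applying body deform corrections.
# Spherical pose readers on both thighs write forward, outward, inward and twist alphas, which drive mostly interpolated weights on parent constraints and additive transforms.
# Grind rail, emote, DBNO and pelvis height and tilt adjust the apron response.
# A shared DeformRig function plus armpit constraints handle body deformation.
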ControlRigBlueprint /BRCosmetics/Characters/Player/Male/Medium/Bodies/M_MED_JazzShoes/Meshes/ClothAsset/M_MED_JazzShoes_ClothEditor_CtrlRig
graphs [
 RigVMFunctionLibrary {links [], nodes []}
 RigVMFunction_Sequence_1_ContainedGraph {nodes [
   n1 = RigVMFunctionEntryNode()
   RigVMFunction_Sequence = &RigVMFunction_Sequence(ExecuteContext=@n1.ExecuteContext)
   n2 = &RigVMFunction_Sequence(ExecuteContext=@RigVMFunction_Sequence.B)
   n3 = &RigVMFunction_Sequence(ExecuteContext=@n2.B)
   n4 = &RigVMFunction_Sequence(ExecuteContext=@n3.B)
   n5 = &RigVMFunction_Sequence(ExecuteContext=@n4.B)
   n6 = &RigVMFunction_Sequence(ExecuteContext=@n5.B)
   RigVMFunctionReturnNode(A=@RigVMFunction_Sequence.A, B=@n2.A, C=@n3.A, D=@n4.A, E=@n5.A, F=@n6.A, G=@n6.B)
  ]}
 RigVMModel {nodes [
   &RigUnit_ParentConstraint(
    AdvancedSettings=(InterpolationType=Average,RotationOrderForFilter=XZY)
    Child=(Type=Bone,Name="dyn_apron_side_l_base")
    ExecutePin=()
    Parents=((Item=(Type=Bone,Name="dyn_apron_fr_l_base"),Weight=0.1),(Item=(Type=Bone,Name="thigh_l"),Weight=0.9))
    Weight=1
   )
   > &RigUnit_ParentConstraint(
    AdvancedSettings=(InterpolationType=Average,RotationOrderForFilter=XZY)
    Child=(Type=Bone,Name="dyn_apron_fr_l_base")
    Parents=((Item=(Type=Bone,Name="dyn_apron_fr_c_base"),Weight=0.1),(Item=(Type=Bone,Name="thigh_l"),Weight=0.9))
    Weight=1
   )
   &RigUnit_ParentConstraint(
    AdvancedSettings=(InterpolationType=Average,RotationOrderForFilter=XZY)
    Child=(Type=Bone,Name="dyn_apron_fr_r_04")
    ExecutePin=()
    Parents=((Item=(Type=Bone,Name="calf_r"),Weight=0.5))
    Weight=0.1
   )
   > &RigUnit_ParentConstraint(
    AdvancedSettings=(InterpolationType=Average,RotationOrderForFilter=XZY)
    Child=(Type=Bone,Name="dyn_apron_fr_l_04")
    Parents=((Item=(Type=Bone,Name="calf_l"),Weight=0.5))
    Weight=0.1
   )
   > &RigUnit_ParentConstraint(
    AdvancedSettings=(InterpolationType=Average,RotationOrderForFilter=XZY)
    Child=(Type=Bone,Name="dyn_apron_side_r_04")
    Parents=((Item=(Type=Bone,Name="calf_r"),Weight=0.5))
    Weight=0.1
   )
   > &RigUnit_ParentConstraint(
    AdvancedSettings=(InterpolationType=Average,RotationOrderForFilter=XZY)
    Child=(Type=Bone,Name="dyn_apron_side_l_04")
    Parents=((Item=(Type=Bone,Name="calf_l"),Weight=0.5))
    Weight=0.1
   )
   n1 = RigVMRerouteNode(Value=())
   n2 = &RigVMFunction_AlphaInterp(
    InterpSpeedDecreasing=30
    InterpSpeedIncreasing=30
    Scale=1
    bClampResult=False
    Value=RigVMVariableNode() {variable=L thighInSide Alpha}.Value
   )
   &RigUnit_ModifyTransforms(
    ItemToModify="((Item=(Type=Bone,Name=\"dyn_apron_side_r_base\"),Transform=(Rotation=(X=0,Y=0.0872,Z=0,W=-0.9962),Translation=(X=0,Y=0,Z=0),Scale3D=(X=1,Y=1,Z=1))),(Item=(Type=Bone,Name=\"dyn_apron_fr_r_base\"),Transform=(Rotation=(X=-0.0298,Y=0.173,Z=0.0819,W=-0.9811),Trans ...(truncated)"
    WeightMinimum=0
    Weight=@n2.Result
    ExecutePin=@n1.Value
   )
   n3 = &RigVMFunction_AlphaInterp(
    InterpSpeedDecreasing=30
    InterpSpeedIncreasing=30
    Scale=1
    bClampResult=False
    Value=RigVMVariableNode() {variable=R thighInSide Alpha}.Value
   )
   > &RigUnit_ModifyTransforms(
    ItemToModify="((Item=(Type=Bone,Name=\"dyn_apron_side_r_base\"),Transform=(Rotation=(X=0,Y=0.0872,Z=0,W=-0.9962),Translation=(X=0,Y=0,Z=0),Scale3D=(X=1,Y=1,Z=1))),(Item=(Type=Bone,Name=\"dyn_apron_fr_r_base\"),Transform=(Rotation=(X=0,Y=0,Z=0,W=-1),Translation=(X=0,Y=0,Z=0) ...(truncated)"
    WeightMinimum=0
    Weight=@n3.Result
   )
   n4 = &RigVMFunction_AlphaInterp(
    InterpSpeedDecreasing=30
    InterpSpeedIncreasing=30
    Scale=1
    bClampResult=False
    Value=RigVMVariableNode() {variable=L thighTwistIn Alpha}.Value
   )
   > &RigUnit_ModifyTransforms(
    ItemToModify="((Item=(Type=Bone,Name=\"dyn_apron_side_l_base\"),Transform=(Rotation=(X=0,Y=0,Z=0,W=1),Translation=(X=0,Y=0,Z=0),Scale3D=(X=1,Y=1,Z=1))),(Item=(Type=Bone,Name=\"dyn_apron_fr_l_base\"),Transform=(Rotation=(X=0,Y=0,Z=0,W=-1),Translation=(X=0,Y=0,Z=0),Scale3D=(X ...(truncated)"
    WeightMinimum=0
    Weight=@n4.Result
   )
   n5 = &RigVMFunction_AlphaInterp(
    InterpSpeedDecreasing=30
    InterpSpeedIncreasing=30
    Scale=1
    bClampResult=False
    Value=RigVMVariableNode() {variable=R thighTwistIn Alpha}.Value
   )
   > &RigUnit_ModifyTransforms(
    ItemToModify="((Item=(Type=Bone,Name=\"dyn_apron_side_r_base\"),Transform=(Rotation=(X=0,Y=0,Z=0,W=-1),Translation=(X=0,Y=0,Z=0),Scale3D=(X=1,Y=1,Z=1))),(Item=(Type=Bone,Name=\"dyn_apron_fr_r_base\"),Transform=(Rotation=(X=0.0448,Y=0.0013,Z=-0.1671,W=-0.9849),Translation=(X ...(truncated)"
    WeightMinimum=0
    Weight=@n5.Result
   )
   RigVMFunction_MathBoolToFloat = &RigVMFunction_MathBoolToFloat(Value=RigVMVariableNode() {variable=Is On Grind Rail}.Value)
   n6 = &RigVMFunction_AlphaInterp(InterpSpeedDecreasing=30, InterpSpeedIncreasing=30, Scale=1, bClampResult=False, Value=@RigVMFunction_MathBoolToFloat.Result)
   > &RigUnit_ParentConstraint(
    AdvancedSettings=(InterpolationType=Average,RotationOrderForFilter=XZY)
    Child=(Type=Bone,Name="dyn_apron_fr_r_01")
    Parents=((Item=(Type=Bone,Name="thigh_r"),Weight=0.5))
    Weight=@n6.Result
   )
   > &RigUnit_ParentConstraint(
    AdvancedSettings=(InterpolationType=Average,RotationOrderForFilter=XZY)
    Child=(Type=Bone,Name="dyn_apron_fr_l_01")
    Parents=((Item=(Type=Bone,Name="thigh_l"),Weight=1))
    Weight=@n6.Result
   )
   > &RigUnit_ModifyTransforms(
    ItemToModify="((Item=(Type=Bone,Name=\"dyn_apron_side_r_base\"),Transform=(Rotation=(X=0,Y=0,Z=-0.0436,W=0.999),Translation=(X=0,Y=0,Z=0),Scale3D=(X=1,Y=1,Z=1))),(Item=(Type=Bone,Name=\"dyn_apron_fr_r_base\"),Transform=(Rotation=(X=0,Y=-0.0436,Z=0,W=0.999),Translation=(X=-0 ...(truncated)"
    WeightMinimum=0
    Weight=@n6.Result
   )
   > &RigUnit_ModifyTransforms(
    ItemToModify="((Item=(Type=Bone,Name=\"dyn_apron_fr_l_base\"),Transform=(Rotation=(X=0,Y=-0.0872,Z=0,W=0.9962),Translation=(X=0,Y=0,Z=0),Scale3D=(X=1,Y=1,Z=1))),(Item=(Type=Bone,Name=\"dyn_apron_fr_l_04\"),Transform=(Rotation=(X=0,Y=0.0872,Z=0,W=-0.9962),Translation=(X=0,Y= ...(truncated)"
    WeightMinimum=0
    Weight=@n6.Result
   )
   n7 = RigVMRerouteNode(Value=())
   n8 = &RigVMFunction_AlphaInterp(
    InterpSpeedDecreasing=30
    InterpSpeedIncreasing=30
    Scale=1
    bClampResult=False
    Value=RigVMVariableNode() {variable=R thighForward Alpha}.Value
   )
   &RigUnit_ModifyTransforms(
    ItemToModify="((Item=(Type=Bone,Name=\"dyn_apron_fr_r_04\"),Transform=(Rotation=(X=0,Y=0,Z=0,W=-1),Translation=(X=0,Y=0,Z=0),Scale3D=(X=1,Y=1,Z=1))),(Item=(Type=Bone,Name=\"dyn_apron_fr_r_base\"),Transform=(Rotation=(X=0,Y=0,Z=0,W=1),Translation=(X=0,Y=0,Z=0),Scale3D=(X=1,Y ...(truncated)"
    WeightMinimum=0
    Weight=@n8.Result
    ExecutePin=@n7.Value
   )
   n9 = RigVMVariableNode() {variable=L thighForward Alpha}
   > &RigUnit_ParentConstraint(
    AdvancedSettings=(InterpolationType=Average,RotationOrderForFilter=XZY)
    Child=(Type=Bone,Name="dyn_apron_fr_c_base")
    Parents=((Item=(Type=Bone,Name="thigh_l"),Weight=0.9),(Item=(Type=Bone,Name="dyn_apron_fr_r_base"),Weight=0.1))
    Weight=@n9.Value
   )
   n10 = RigVMVariableNode() {variable=R thighForward Alpha}
   > &RigUnit_ParentConstraint(
    AdvancedSettings=(InterpolationType=Average,RotationOrderForFilter=XZY)
    Child=(Type=Bone,Name="dyn_apron_fr_c_base")
    Parents=((Item=(Type=Bone,Name="thigh_r"),Weight=0.9),(Item=(Type=Bone,Name="dyn_apron_fr_r_base"),Weight=0.1))
    Weight=@n10.Value
   )
   n11 = RigVMVariableNode() {variable=R thighForward Alpha}
   n12 = RigVMVariableNode() {variable=L thighForward Alpha}
   Subtract = &RigVMFunction_MathDoubleSub(B=@n11.Value, A=@n12.Value)
   n13 = &RigUnit_GetRelativeTransformForItem(Child=(Type=Bone,Name="pelvis"), Parent=(Type=Bone,Name="root"), bChildInitial=False, bParentInitial=false)
   n14 = RigVMDispatchNode(Struct=@n13.RelativeTransform.Translation) {template="DISPATCH_RigVMDispatch_BreakStruct(in Struct,out Elements)"}
   Absolute = &RigVMFunction_MathDoubleAbs(Value=&RigVMFunction_MathQuaternionToEuler(RotationOrder=ZYX, Value=@n13.RelativeTransform.Rotation).Result.Y)
   n15 = RigVMVariableNode() {variable=isDBNO}
   RigVMFunction_MathFloatMake = /Script/RigVM.RigVMFunction_MathFloatMake(Value=&RigVMFunction_MathDoubleSub(A=1, B=&RigVMFunction_MathBoolToFloat(Value=@n15.Value).Result).Result) {is=struct RigVMUnitNode}
   n16 = &RigVMFunction_MathDoubleMul(
    A=&RigVMFunction_MathDoubleMul(
     B=&RigVMFunction_MathDoubleRemap(SourceMaximum=45, SourceMinimum=65, TargetMaximum=1, TargetMinimum=0, Value=@n14.Elements.Z).Result
     A=&RigVMFunction_MathDoubleRemap(SourceMaximum=69, SourceMinimum=80, TargetMaximum=1, TargetMinimum=0, Value=@Absolute.Result).Result
    ).Result
    B=@RigVMFunction_MathFloatMake.Value
   )
   n17 = &RigVMFunction_AlphaInterp(
    InterpSpeedDecreasing=30
    InterpSpeedIncreasing=30
    Scale=1
    bClampResult=true
    Value=&RigVMFunction_MathDoubleSub(A=&RigVMFunction_MathDoubleMul(B=0.5, A=@Subtract.Result).Result, B=@n16.Result).Result
   )
   > &RigUnit_ParentConstraint(
    AdvancedSettings=(InterpolationType=Average,RotationOrderForFilter=XZY)
    Child=(Type=Bone,Name="dyn_apron_fr_r_base")
    Parents=((Item=(Type=Bone,Name="dyn_apron_fr_c_base"),Weight=0.1))
    Weight=@n17.Result
   )
   n18 = &RigVMFunction_AlphaInterp(
    InterpSpeedDecreasing=30
    InterpSpeedIncreasing=30
    Scale=1
    bClampResult=true
    Value=&RigVMFunction_MathDoubleSub(A=&RigVMFunction_MathDoubleMul(B=0.35, A=@Subtract.Result).Result, B=@n16.Result).Result
   )
   > &RigUnit_ParentConstraint(
    AdvancedSettings=(InterpolationType=Average,RotationOrderForFilter=XZY)
    Child=(Type=Bone,Name="dyn_apron_side_r_base")
    Parents=((Item=(Type=Bone,Name="dyn_apron_fr_r_base"),Weight=0.1))
    Weight=@n18.Result
   )
   n19 = &RigVMFunction_MathDoubleSub(A=@n11.Value, B=@n12.Value)
   Multiply = &RigVMFunction_MathDoubleMul(B=0.5, A=@n19.Result)
   n20 = &RigVMFunction_AlphaInterp(
    InterpSpeedDecreasing=30
    InterpSpeedIncreasing=30
    Scale=1
    bClampResult=true
    Value=&RigVMFunction_MathDoubleSub(A=@Multiply.Result, B=@n16.Result).Result
   )
   > &RigUnit_ParentConstraint(
    AdvancedSettings=(InterpolationType=Average,RotationOrderForFilter=XZY)
    Child=(Type=Bone,Name="dyn_apron_fr_l_base")
    Parents=((Item=(Type=Bone,Name="dyn_apron_fr_c_base"),Weight=0.1))
    Weight=@n20.Result
   )
   n21 = &RigVMFunction_AlphaInterp(
    InterpSpeedDecreasing=30
    InterpSpeedIncreasing=30
    Scale=1
    bClampResult=False
    Value=&RigVMFunction_MathDoubleSub(A=&RigVMFunction_MathDoubleMul(B=0.35, A=@n19.Result).Result, B=@n16.Result).Result
   )
   > &RigUnit_ParentConstraint(
    AdvancedSettings=(InterpolationType=Average,RotationOrderForFilter=XZY)
    Child=(Type=Bone,Name="dyn_apron_side_l_base")
    Parents=((Item=(Type=Bone,Name="dyn_apron_fr_l_base"),Weight=0.1))
    Weight=@n21.Result
   )
   Clamp = &RigVMFunction_MathDoubleClamp(Value=@Subtract.Result)
   Add = &RigVMFunction_MathDoubleAdd(A=@Clamp.Result, B=&RigVMFunction_MathDoubleClamp(Value=@n19.Result).Result)
   n22 = &RigVMFunction_AlphaInterp(InRange=(Min=1,Max=0), InterpSpeedDecreasing=30, InterpSpeedIncreasing=30, Scale=1, bClampResult=true, bMapRange=true, Value=@Add.Result)
   > &RigUnit_ParentConstraint(
    AdvancedSettings=(InterpolationType=Average,RotationOrderForFilter=XZY)
    Child=(Type=Bone,Name="dyn_apron_fr_c_01")
    Parents=((Item=(Type=Bone,Name="dyn_apron_fr_l_01"),Weight=0.5),(Item=(Type=Bone,Name="dyn_apron_fr_r_01"),Weight=0.5))
    Weight=@n22.Result
   )
   n23 = &RigVMFunction_AlphaInterp(
    InterpSpeedDecreasing=30
    InterpSpeedIncreasing=30
    Scale=1
    bClampResult=False
    Value=RigVMVariableNode() {variable=L thighForward Alpha}.Value
   )
   > &RigUnit_ModifyTransforms(
    ItemToModify="((Item=(Type=Bone,Name=\"dyn_apron_side_l_base\"),Transform=(Rotation=(X=-0.171,Y=0.0302,Z=0,W=-0.9848),Translation=(X=0,Y=0,Z=0),Scale3D=(X=1,Y=1,Z=1))),(Item=(Type=Bone,Name=\"dyn_apron_fr_l_base\"),Transform=(Rotation=(X=0,Y=0,Z=0.0872,W=0.9962),Translation ...(truncated)"
    WeightMinimum=0
    Weight=@n23.Result
   )
   n24 = RigVMVariableNode() {variable=R thighForward Alpha}
   n25 = RigVMVariableNode() {variable=L thighForward Alpha}
   n26 = &RigVMFunction_AlphaInterp(
    InterpSpeedDecreasing=30
    InterpSpeedIncreasing=30
    Scale=1
    bClampResult=False
    Value=&RigVMFunction_MathDoubleClamp(Value=&RigVMFunction_MathDoubleMul(B=@n24.Value, A=@n25.Value).Result).Result
   )
   > &RigUnit_ModifyTransforms(
    ItemToModify=((Item=(Type=Bone,Name="dyn_apron_fr_c_02"),Transform=(Rotation=(X=-0.0541,Y=0.2015,Z=-0.0299,W=0.9775),Translation=(X=0,Y=0,Z=0),Scale3D=(X=1,Y=1,Z=1))))
    WeightMinimum=0
    Weight=@n26.Result
   )
   n27 = &RigVMFunction_AlphaInterp(
    InterpSpeedDecreasing=30
    InterpSpeedIncreasing=30
    Scale=1
    bClampResult=False
    Value=RigVMVariableNode() {variable=L thighOutSide Alpha}.Value
   )
   > n28 = &RigUnit_ModifyTransforms(
    ItemToModify="((Item=(Type=Bone,Name=\"dyn_apron_side_l_base\"),Transform=(Rotation=(X=0,Y=0.0872,Z=0,W=-0.9962),Translation=(X=0,Y=0,Z=0),Scale3D=(X=1,Y=1,Z=1))),(Item=(Type=Bone,Name=\"dyn_apron_side_l_03\"),Transform=(Rotation=(X=0,Y=-0.0872,Z=0,W=-0.9962),Translation=(X ...(truncated)"
    WeightMinimum=0
    Weight=@n27.Result
   )
   RigVMFunction_ControlFlowBranch = &RigVMFunction_ControlFlowBranch(ExecuteContext=@n28.ExecutePin, Condition=RigVMVariableNode() {variable=Is Playing Emote}.Value)
   n29 = &RigVMFunction_AlphaInterp(
    InterpSpeedDecreasing=25
    InterpSpeedIncreasing=25
    Scale=1
    bClampResult=False
    Value=RigVMVariableNode() {variable=L thighOutSide Alpha}.Value
   )
   &RigUnit_ModifyTransforms(
    ItemToModify="((Item=(Type=Bone,Name=\"dyn_apron_side_l_base\"),Transform=(Rotation=(X=0,Y=-0.2588,Z=0,W=0.9659),Translation=(X=-0.4434,Y=-0.8339,Z=0),Scale3D=(X=1,Y=1,Z=1))),(Item=(Type=Bone,Name=\"dyn_apron_fr_l_base\"),Transform=(Rotation=(X=0,Y=-0.1305,Z=0,W=0.9914),Tra ...(truncated)"
    WeightMinimum=0
    Weight=@n29.Result
    ExecutePin=@RigVMFunction_ControlFlowBranch.True
   )
   n30 = &RigVMFunction_AlphaInterp(
    InterpSpeedDecreasing=25
    InterpSpeedIncreasing=25
    Scale=1
    bClampResult=False
    Value=RigVMVariableNode() {variable=R thighOutSide Alpha}.Value
   )
   > &RigUnit_ModifyTransforms(
    ItemToModify="((Item=(Type=Bone,Name=\"dyn_apron_side_r_base\"),Transform=(Rotation=(X=0,Y=-0.0872,Z=0,W=0.9962),Translation=(X=0,Y=0,Z=0),Scale3D=(X=1,Y=1,Z=1))),(Item=(Type=Bone,Name=\"dyn_apron_fr_r_base\"),Transform=(Rotation=(X=0,Y=-0.0872,Z=0,W=0.9962),Translation=(X= ...(truncated)"
    WeightMinimum=0
    Weight=@n30.Result
   )
   n31 = RigVMRerouteNode(Value=())
   n32 = &RigUnit_SphericalPoseReader(
    ActiveRegionScaleFactors=(PositiveWidth=4,NegativeWidth=1,PositiveHeight=2,NegativeHeight=1.5)
    ActiveRegionSize=0.1
    Debug=(bDrawDebug=false,bDraw2D=False,bDrawLocalAxes=False,DebugScale=25,DebugSegments=20,DebugThickness=0.25)
    DriverAxis=(X=-1,Y=0,Z=0)
    DriverItem=(Type=Bone,Name="thigh_l")
    FalloffRegionScaleFactors=(PositiveWidth=1,NegativeWidth=0.7,PositiveHeight=1,NegativeHeight=1)
    FalloffSize=0.4
    FlipHeightScaling=False
    FlipWidthScaling=False
    OptionalParentItem=(Type=Bone,Name="None")
    RotationOffset=(X=108.8353,Y=0,Z=13.2741)
    ExecutePin=@n31.Value
   )
   > n33 = &RigUnit_SphericalPoseReader(
    ActiveRegionScaleFactors=(PositiveWidth=4,NegativeWidth=1,PositiveHeight=2,NegativeHeight=1.5)
    ActiveRegionSize=0.1
    Debug=(bDrawDebug=false,bDraw2D=False,bDrawLocalAxes=False,DebugScale=25,DebugSegments=20,DebugThickness=0.25)
    DriverAxis=(X=1,Y=0,Z=0)
    DriverItem=(Type=Bone,Name="thigh_r")
    FalloffRegionScaleFactors=(PositiveWidth=1,NegativeWidth=0.7,PositiveHeight=1,NegativeHeight=1)
    FalloffSize=0.4
    FlipHeightScaling=False
    FlipWidthScaling=true
    OptionalParentItem=(Type=Bone,Name="None")
    RotationOffset=(X=-67.4691,Y=0,Z=14.7473)
   )
   > n34 = &RigUnit_SphericalPoseReader(
    ActiveRegionScaleFactors=(PositiveWidth=3.5,NegativeWidth=2,PositiveHeight=2,NegativeHeight=2)
    ActiveRegionSize=0.1
    Debug=(bDrawDebug=true,bDraw2D=False,bDrawLocalAxes=False,DebugScale=25,DebugSegments=20,DebugThickness=0.25)
    DriverAxis=(X=-1,Y=0,Z=0)
    DriverItem=(Type=Bone,Name="thigh_l")
    FalloffRegionScaleFactors=(PositiveWidth=1,NegativeWidth=2,PositiveHeight=1,NegativeHeight=1)
    FalloffSize=0.3
    FlipHeightScaling=false
    FlipWidthScaling=false
    OptionalParentItem=(Type=Bone,Name="None")
    RotationOffset=(X=157.2715,Y=-17.1205,Z=23.0518)
   )
   > n35 = &RigUnit_SphericalPoseReader(
    ActiveRegionScaleFactors=(PositiveWidth=3.5,NegativeWidth=2,PositiveHeight=3,NegativeHeight=3)
    ActiveRegionSize=0.1
    Debug=(bDrawDebug=true,bDraw2D=False,bDrawLocalAxes=False,DebugScale=25,DebugSegments=20,DebugThickness=0.25)
    DriverAxis=(X=1,Y=0,Z=0)
    DriverItem=(Type=Bone,Name="thigh_r")
    FalloffRegionScaleFactors=(PositiveWidth=1,NegativeWidth=1,PositiveHeight=1,NegativeHeight=1)
    FalloffSize=0.3
    FlipHeightScaling=False
    FlipWidthScaling=true
    OptionalParentItem=(Type=Bone,Name="None")
    RotationOffset=(X=1.6393,Y=-20,Z=0)
   )
   > n36 = &RigUnit_SphericalPoseReader(
    ActiveRegionScaleFactors=(PositiveWidth=2,NegativeWidth=1,PositiveHeight=1.2,NegativeHeight=1)
    ActiveRegionSize=0.2
    Debug=(bDrawDebug=false,bDraw2D=False,bDrawLocalAxes=False,DebugScale=25,DebugSegments=20,DebugThickness=0.25)
    DriverAxis=(X=-1,Y=0,Z=0)
    DriverItem=(Type=Bone,Name="thigh_l")
    FalloffRegionScaleFactors=(PositiveWidth=1.5,NegativeWidth=1,PositiveHeight=2,NegativeHeight=1)
    FalloffSize=0.3
    FlipHeightScaling=false
    FlipWidthScaling=false
    OptionalParentItem=(Type=Bone,Name="None")
    RotationOffset=(X=40.8051,Y=35.9559,Z=-11.364)
   )
   > n37 = &RigUnit_SphericalPoseReader(
    ActiveRegionScaleFactors=(PositiveWidth=1,NegativeWidth=1,PositiveHeight=1,NegativeHeight=1)
    ActiveRegionSize=0.2
    Debug=(bDrawDebug=false,bDraw2D=False,bDrawLocalAxes=False,DebugScale=25,DebugSegments=20,DebugThickness=0.25)
    DriverAxis=(X=1,Y=0,Z=0)
    DriverItem=(Type=Bone,Name="thigh_r")
    FalloffRegionScaleFactors=(PositiveWidth=1,NegativeWidth=0.5,PositiveHeight=1,NegativeHeight=1)
    FalloffSize=0.3
    FlipHeightScaling=False
    FlipWidthScaling=true
    OptionalParentItem=(Type=Bone,Name="None")
    RotationOffset=(X=169.1534,Y=48.0832,Z=-15.1)
   )
   > n38 = &RigUnit_SphericalPoseReader(
    ActiveRegionScaleFactors=(PositiveWidth=1,NegativeWidth=1,PositiveHeight=1.2,NegativeHeight=1)
    ActiveRegionSize=0.15
    Debug=(bDrawDebug=true,bDraw2D=False,bDrawLocalAxes=False,DebugScale=25,DebugSegments=20,DebugThickness=0.25)
    DriverAxis=(X=0,Y=1,Z=0)
    DriverItem=(Type=Bone,Name="thigh_l")
    FalloffRegionScaleFactors=(PositiveWidth=1,NegativeWidth=1,PositiveHeight=1,NegativeHeight=1)
    FalloffSize=0.3
    FlipHeightScaling=false
    FlipWidthScaling=false
    OptionalParentItem=(Type=Bone,Name="None")
    RotationOffset=(X=21.5922,Y=-16.4671,Z=2.9545)
   )
   > n39 = &RigUnit_SphericalPoseReader(
    ActiveRegionScaleFactors=(PositiveWidth=1,NegativeWidth=1,PositiveHeight=1,NegativeHeight=1)
    ActiveRegionSize=0.15
    Debug=(bDrawDebug=true,bDraw2D=False,bDrawLocalAxes=False,DebugScale=25,DebugSegments=20,DebugThickness=0.25)
    DriverAxis=(X=0,Y=-1,Z=0)
    DriverItem=(Type=Bone,Name="thigh_r")
    FalloffRegionScaleFactors=(PositiveWidth=1,NegativeWidth=0.5,PositiveHeight=1,NegativeHeight=1)
    FalloffSize=0.3
    FlipHeightScaling=False
    FlipWidthScaling=true
    OptionalParentItem=(Type=Bone,Name="None")
    RotationOffset=(X=195.3337,Y=-4.9916,Z=0)
   )
   n40 = RigVMRerouteNode(Value=())
   RigVMVariableNode(ExecuteContext=@n40.Value, Value=@n32.OutputParam) {variable=L thighForward Alpha}
   > RigVMVariableNode(Value=@n33.OutputParam) {variable=R thighForward Alpha}
   > RigVMVariableNode(Value=@n34.OutputParam) {variable=L thighOutSide Alpha}
   > RigVMVariableNode(Value=@n35.OutputParam) {variable=R thighOutSide Alpha}
   > RigVMVariableNode(Value=@n36.OutputParam) {variable=L thighInSide Alpha}
   > RigVMVariableNode(Value=@n37.OutputParam) {variable=R thighInSide Alpha}
   > RigVMVariableNode(Value=@n38.OutputParam) {variable=L thighTwistIn Alpha}
   > RigVMVariableNode(Value=@n39.OutputParam) {variable=R thighTwistIn Alpha}
   RigVMFunction_MathQuaternionToEuler = &RigVMFunction_MathQuaternionToEuler(
    RotationOrder=ZYX
    Value=&RigUnit_GetRelativeTransformForItem(Child=(Type=Bone,Name="calf_l"), Parent=(Type=Bone,Name="thigh_l"), bChildInitial=False, bParentInitial=False).RelativeTransform.Rotation
   )
   Remap = &RigVMFunction_MathDoubleRemap(SourceMaximum=100, SourceMinimum=45, TargetMaximum=1, TargetMinimum=0, Value=@RigVMFunction_MathQuaternionToEuler.Result.Z)
   > RigVMVariableNode(Value=@Remap.Result) {variable=L calfBend Alpha}
   n41 = &RigUnit_GetRelativeTransformForItem(Child=(Type=Bone,Name="calf_r"), Parent=(Type=Bone,Name="thigh_r"), bChildInitial=False, bParentInitial=False)
   > RigVMVariableNode(Value=&RigVMFunction_MathDoubleRemap(
     SourceMaximum=100
     SourceMinimum=45
     TargetMaximum=1
     TargetMinimum=0
     Value=&RigVMFunction_MathQuaternionToEuler(RotationOrder=ZYX, Value=@n41.RelativeTransform.Rotation).Result.Z
    ).Result) {variable=R calfBend Alpha}
   n42 = RigVMRerouteNode(Value=())
   &RigUnit_ModifyTransforms(
    ItemToModify=((Item=(Type=Bone,Name="dyn_apron_fr_r_base"),Transform=(Rotation=(X=0,Y=0,Z=0,W=1),Translation=(X=0,Y=0,Z=0),Scale3D=(X=1,Y=1,Z=1))))
    Weight=1
    WeightMinimum=0
    ExecutePin=@n42.Value
   )
   > &RigUnit_ParentConstraint(
    AdvancedSettings=(InterpolationType=Average,RotationOrderForFilter=XZY)
    Child=(Type=Bone,Name="dyn_apron_fr_c_base")
    Parents=((Item=(Type=Bone,Name="dyn_apron_fr_l_base"),Weight=0.5),(Item=(Type=Bone,Name="dyn_apron_fr_r_base"),Weight=0.5))
    Weight=1
   )
   > &RigUnit_ParentConstraint(
    AdvancedSettings=(InterpolationType=Average,RotationOrderForFilter=XZY)
    Child=(Type=Bone,Name="dyn_apron_side_l_base")
    Parents=((Item=(Type=Bone,Name="dyn_apron_fr_l_base"),Weight=1))
    Weight=0.3
   )
   > &RigUnit_ParentConstraint(
    AdvancedSettings=(InterpolationType=Average,RotationOrderForFilter=XZY)
    Child=(Type=Bone,Name="dyn_apron_side_r_base")
    Parents=((Item=(Type=Bone,Name="dyn_apron_fr_r_base"),Weight=1))
    Weight=0.3
   )
   RigUnit_BeginExecution = &RigUnit_BeginExecution()
   RigVMFunction_Sequence = RigVMAggregateNode(ExecuteContext=@RigUnit_BeginExecution.ExecutePin)
   FN_DeformRig_FNC = RigVMFunctionReferenceNode(
    DeformRig_ Curve="(deformRig_l_upperarm_up_alpha_72_#1=False,deformRig_r_upperarm_up_alpha_73_#2=False,deformRig_l_upperarm_fr_alpha_74_#3=False,deformRig_r_upperarm_fr_alpha_75_CB51B8 ...(truncated)"
    DeformRig_Bool="(upperarm_1_#4=True,elbow_3_#5=false,wrist_7_#6=false,groin_8_#7=false,glute_17_#8=false,knee_9_40C408324C ...(truncated)"
    DeformRig_Debug=(upperarm_debug_18_#4=False,wrist_debug_20_#6=False,lat_debug_26_#9=False,groinGlute_debug_27_#7=False)
    DeformRig_Value="(upperarm_up_18_#10=(Rotation=(X=0,Y=0,Z=0,W=1),Translation=(X=0,Y=0,Z=8),Scale3D=(X=0.7,Y=1,Z=1)),upperarm_fwd_fr_19_#11=(Rotation=(X=0,Y=0,Z=0,W=1),Translation=(X=0,Y=10,Z=0),Scale3D=(X=0.5,Y=1,Z= ...(truncated)"
    ExecuteContext=@RigVMFunction_Sequence.A
   )
   > &RigUnit_ParentConstraint(
    AdvancedSettings=(InterpolationType=Average,RotationOrderForFilter=XZY)
    Child=(Type=Bone,Name="deform_upperarm_armpit_r")
    Parents=((Item=(Type=Bone,Name="clavicle_r"),Weight=0.5),(Item=(Type=Bone,Name="upperarm_twist_01_r"),Weight=0.5))
    Weight=1
   )
   > &RigUnit_ParentConstraint(
    AdvancedSettings=(InterpolationType=Average,RotationOrderForFilter=XZY)
    Child=(Type=Bone,Name="deform_upperarm_armpit_l")
    Parents=((Item=(Type=Bone,Name="clavicle_l"),Weight=0.5),(Item=(Type=Bone,Name="upperarm_twist_01_l"),Weight=0.5))
    Weight=1
   )
   > &RigUnit_ModifyTransforms(
    ItemToModify=((Item=(Type=Bone,Name="deform_upperarm_armpit_l"),Transform=(Rotation=(X=0,Y=0,Z=0,W=1),Translation=(X=-4,Y=0,Z=-0.4424),Scale3D=(X=1,Y=1,Z=1))))
    WeightMinimum=0
    Weight=RigVMDispatchNode(BoneName=root, Default=0, Name=deformRig_l_upperarm_up_alpha) {template="DISPATCH_RigDispatch_GetAnimAttribute(in Name,in BoneName,in Default,out Value,out Found)"}.Value
   )
   > &RigUnit_ModifyTransforms(
    ItemToModify=((Item=(Type=Bone,Name="deform_upperarm_armpit_r"),Transform=(Rotation=(X=0,Y=0,Z=0,W=1),Translation=(X=0,Y=4,Z=-0.5738),Scale3D=(X=1,Y=1,Z=1))))
    WeightMinimum=0
    Weight=RigVMDispatchNode(BoneName=root, Default=0, Name=deformRig_r_upperarm_up_alpha) {template="DISPATCH_RigDispatch_GetAnimAttribute(in Name,in BoneName,in Default,out Value,out Found)"}.Value
   )
   &RigVMFunction_AlphaInterp(InterpSpeedDecreasing=30, InterpSpeedIncreasing=30, Scale=1, bClampResult=False, Value=@n9.Value)
   &RigVMFunction_AlphaInterp(InterpSpeedDecreasing=30, InterpSpeedIncreasing=30, Scale=1, bClampResult=False, Value=@n10.Value)
   RigVMVariableNode() {variable=R thighOutSide Alpha}
  ]}
]
hierarchy [
 root {loc=<0 0 0>, rot=<0 0 0 1>}
 pelvis {parent=root, loc=<0 -0.7071 93.1191>, rot=<0.0127 0.707 -0.0127 -0.707>}
 spine_01 {parent=pelvis, loc=<6.7494 -0.6388 0>, rot=<0 0 0.034 -0.9994>}
 spine_02 {parent=spine_01, loc=<7.252 -0.035 0>, rot=<0 0 0.0238 -0.9997>}
 spine_03 {parent=spine_02, loc=<6.0106 0 0>, rot=<0 0 0.0149 -0.9999>}
 spine_04 {parent=spine_03, loc=<8.1023 0 0>, rot=<0 0 -0.0132 -0.9999>}
 spine_05 {parent=spine_04, loc=<10.0142 0 0>, rot=<0 0 -0.0015 -1>}
 attach_backpack {parent=spine_05, loc=<-0.123 -13.473 0>, rot=<-0.0093 -0.707 0.0093 -0.707>}
 neck_01 {parent=spine_05, loc=<11.3138 0.1529 0>, rot=<0 0 -0.2003 -0.9797>}
 neck_02 {parent=neck_01, loc=<4.5627 0 0>, rot=<0 0 0.0504 -0.9987>}
 head {parent=neck_02, loc=<5.3914 0 0>, rot=<0 0 0.115 -0.9934>}
 clavicle_l {parent=spine_05, loc=<4.3575 -0.1466 -2.3223>, rot=<0.0465 -0.7289 -0.009 -0.6829>}
 upperarm_l {parent=clavicle_l, loc=<15.5111 0 0>, rot=<0.0226 -0.3234 -0.0088 -0.9459>}
 lowerarm_l {parent=upperarm_l, loc=<26.6514 0 0>, rot=<0.0391 -0.0099 -0.2453 -0.9686>}
 hand_l {parent=lowerarm_l, loc=<22.3491 0 0>, rot=<-0.7081 -0.0507 0.0975 -0.6975>}
 thumb_01_l {parent=hand_l, loc=<1.3118 -0.3191 -2.459>, rot=<0.6546 -0.4042 0.0664 -0.6354>}
 thumb_02_l {parent=thumb_01_l, loc=<5.9336 -0.0448 0.0401>, rot=<-0.0037 0.0213 0.298 -0.9543>}
 thumb_03_l {parent=thumb_02_l, loc=<2.2223 0.3682 -0.0907>, rot=<0.0118 -0.013 0.0066 -0.9998>}
 index_metacarpal_l {parent=hand_l, loc=<1.6535 0.0183 -2.96>, rot=<0.0158 -0.0124 -0.0173 -0.9996>}
 index_01_l {parent=index_metacarpal_l, loc=<8.3008 0.0357 -0.0524>, rot=<-0.0044 0.0485 0.0924 -0.9945>}
 index_02_l {parent=index_01_l, loc=<3.3587 -0.3004 0.0452>, rot=<-0.002 0.0027 0.3225 -0.9466>}
 index_03_l {parent=index_02_l, loc=<2.5578 -0.123 -0.0252>, rot=<-0.002 0.0067 0.2558 -0.9667>}
 middle_metacarpal_l {parent=hand_l, loc=<2.1856 -0.1725 -0.5984>, rot=<0.006 0.0168 -0.0329 -0.9993>}
 middle_01_l {parent=middle_metacarpal_l, loc=<7.3275 -0.2447 0.0129>, rot=<-0.0208 0.0042 0.0858 -0.9961>}
 middle_02_l {parent=middle_01_l, loc=<4.0123 0.1401 0.1175>, rot=<-0.0218 0.0258 0.3512 -0.9357>}
 middle_03_l {parent=middle_02_l, loc=<2.6077 0.0534 0.1373>, rot=<0.0082 0.0188 0.1708 -0.9851>}
 ring_metacarpal_l {parent=hand_l, loc=<2.2074 -0.3625 1.3638>, rot=<-0.0284 0.0722 -0.0458 -0.9959>}
 ring_01_l {parent=ring_metacarpal_l, loc=<7.0295 -0.6197 -0.1446>, rot=<-0.0489 -0.0252 0.1892 -0.9804>}
 ring_02_l {parent=ring_01_l, loc=<3.6174 0.4647 -0.1027>, rot=<0 0 0.2806 -0.9598>}
 ring_03_l {parent=ring_02_l, loc=<2.4152 0.114 0.0199>, rot=<0 0 0.1189 -0.9929>}
 pinky_metacarpal_l {parent=hand_l, loc=<1.7616 -0.9753 3.3814>, rot=<-0.0397 0.0814 -0.069 -0.9935>}
 pinky_01_l {parent=pinky_metacarpal_l, loc=<7.1709 -0.4831 -0.1094>, rot=<-0.0654 -0.0035 0.2242 -0.9723>}
 pinky_02_l {parent=pinky_01_l, loc=<2.8274 0.0893 -0.1043>, rot=<0.0037 0.0011 0.2854 -0.9584>}
 pinky_03_l {parent=pinky_02_l, loc=<2.2046 0.0307 -0.0078>, rot=<-0.0002 0.0098 0.1445 -0.9895>}
 lowerarm_twist_02_l {parent=lowerarm_l, loc=<9.6199 0 0>, rot=<0 0 0 1>}
 lowerarm_twist_01_l {parent=lowerarm_l, loc=<17.8549 0 0>, rot=<0 0 0 1>}
 upperarm_twist_01_l {parent=upperarm_l, loc=<0.0041 0 0>, rot=<0 0 0 1>}
 deform_upperarm_bk_base_l {parent=upperarm_twist_01_l, loc=<-3.2582 -5.546 -2.0366>, rot=<0 -0.0007 -0.0191 -0.9998>}
 deform_upperarm_bk_l {parent=deform_upperarm_bk_base_l, loc=<0 0 0>, rot=<0 0 0 1>}
 deform_upperarm_fr_base_l {parent=upperarm_twist_01_l, loc=<-2.8462 3.9432 -2.0158>, rot=<0 -0.0007 -0.0191 -0.9998>}
 deform_upperarm_fr_l {parent=deform_upperarm_fr_base_l, loc=<0 0 0>, rot=<0 0 0 1>}
 deform_upperarm_up_base_l {parent=upperarm_twist_01_l, loc=<-5.9223 -0.7874 1.605>, rot=<0 -0.0007 -0.0191 -0.9998>}
 deform_upperarm_up_l {parent=deform_upperarm_up_base_l, loc=<0 0 0>, rot=<0 0 0 1>}
 deform_upperarm_armpit_base_l {parent=upperarm_twist_01_l, loc=<8.4485 1.3804 -10.6945>, rot=<-0.0134 0.3555 -0.014 -0.9345>}
 deform_upperarm_armpit_l {parent=deform_upperarm_armpit_base_l, loc=<0 0 0>, rot=<0 0 0 1>}
 upperarm_twist_02_l {parent=upperarm_l, loc=<13.7941 0 0>, rot=<0 0 0 1>}
 attach_cape {parent=spine_05, loc=<9.017 -11.714 0>, rot=<0.0616 -0.7044 -0.0616 -0.7044>}
 clavicle_r {parent=spine_05, loc=<4.3575 -0.1466 2.3223>, rot=<0.7289 0.0465 -0.6829 0.009>}
 upperarm_r {parent=clavicle_r, loc=<-15.5111 0 0>, rot=<0.0226 -0.3234 -0.0088 -0.9459>}
 lowerarm_r {parent=upperarm_r, loc=<-26.6514 0 0>, rot=<0.0391 -0.0099 -0.2453 -0.9686>}
 hand_r {parent=lowerarm_r, loc=<-22.3491 0 0>, rot=<-0.7081 -0.0507 0.0975 -0.6975>}
 thumb_01_r {parent=hand_r, loc=<-1.3118 0.3191 2.459>, rot=<0.6546 -0.4042 0.0664 -0.6354>}
 thumb_02_r {parent=thumb_01_r, loc=<-5.9336 0.0448 -0.0401>, rot=<-0.0037 0.0213 0.298 -0.9543>}
 thumb_03_r {parent=thumb_02_r, loc=<-2.2223 -0.3682 0.0907>, rot=<0.0118 -0.013 0.0066 -0.9998>}
 index_metacarpal_r {parent=hand_r, loc=<-1.6535 -0.0183 2.96>, rot=<0.0158 -0.0124 -0.0173 -0.9996>}
 index_01_r {parent=index_metacarpal_r, loc=<-8.3008 -0.0357 0.0524>, rot=<-0.0044 0.0485 0.0924 -0.9945>}
 index_02_r {parent=index_01_r, loc=<-3.3587 0.3004 -0.0452>, rot=<-0.002 0.0027 0.3225 -0.9466>}
 index_03_r {parent=index_02_r, loc=<-2.5578 0.123 0.0252>, rot=<-0.002 0.0067 0.2558 -0.9667>}
 middle_metacarpal_r {parent=hand_r, loc=<-2.1856 0.1725 0.5984>, rot=<0.006 0.0168 -0.0329 -0.9993>}
 middle_01_r {parent=middle_metacarpal_r, loc=<-7.3275 0.2447 -0.0129>, rot=<-0.0208 0.0042 0.0858 -0.9961>}
 middle_02_r {parent=middle_01_r, loc=<-4.0123 -0.1401 -0.1175>, rot=<-0.0218 0.0258 0.3512 -0.9357>}
 middle_03_r {parent=middle_02_r, loc=<-2.6077 -0.0534 -0.1373>, rot=<0.0082 0.0188 0.1708 -0.9851>}
 ring_metacarpal_r {parent=hand_r, loc=<-2.2074 0.3625 -1.3638>, rot=<-0.0284 0.0722 -0.0458 -0.9959>}
 ring_01_r {parent=ring_metacarpal_r, loc=<-7.0295 0.6197 0.1446>, rot=<-0.0489 -0.0252 0.1892 -0.9804>}
 ring_02_r {parent=ring_01_r, loc=<-3.6174 -0.4647 0.1027>, rot=<0 0 0.2806 -0.9598>}
 ring_03_r {parent=ring_02_r, loc=<-2.4152 -0.114 -0.0199>, rot=<0 0 0.1189 -0.9929>}
 pinky_metacarpal_r {parent=hand_r, loc=<-1.7616 0.9753 -3.3814>, rot=<-0.0397 0.0814 -0.069 -0.9935>}
 pinky_01_r {parent=pinky_metacarpal_r, loc=<-7.1709 0.4831 0.1094>, rot=<-0.0654 -0.0035 0.2242 -0.9723>}
 pinky_02_r {parent=pinky_01_r, loc=<-2.8274 -0.0893 0.1043>, rot=<0.0037 0.0011 0.2854 -0.9584>}
 pinky_03_r {parent=pinky_02_r, loc=<-2.2046 -0.0307 0.0078>, rot=<-0.0002 0.0098 0.1445 -0.9895>}
 lowerarm_twist_02_r {parent=lowerarm_r, loc=<-9.6199 0 0>, rot=<0 0 0 1>}
 lowerarm_twist_01_r {parent=lowerarm_r, loc=<-17.8549 0 0>, rot=<0 0 0 1>}
 upperarm_twist_01_r {parent=upperarm_r, loc=<-0.0041 0 0>, rot=<0 0 0 1>}
 deform_upperarm_bk_base_r {parent=upperarm_twist_01_r, loc=<3.2582 5.546 2.0366>, rot=<0 -0.0007 -0.0191 -0.9998>}
 deform_upperarm_bk_r {parent=deform_upperarm_bk_base_r, loc=<0 0 0>, rot=<0 0 0 1>}
 deform_upperarm_fr_base_r {parent=upperarm_twist_01_r, loc=<2.8463 -3.9433 2.0158>, rot=<0 -0.0007 -0.0191 -0.9998>}
 deform_upperarm_fr_r {parent=deform_upperarm_fr_base_r, loc=<0 0 0>, rot=<0 0 0 1>}
 deform_upperarm_up_base_r {parent=upperarm_twist_01_r, loc=<5.9223 0.7874 -1.605>, rot=<0 -0.0007 -0.0191 -0.9998>}
 deform_upperarm_up_r {parent=deform_upperarm_up_base_r, loc=<0 0 0>, rot=<0 0 0 1>}
 deform_upperarm_armpit_base_r {parent=upperarm_twist_01_r, loc=<-8.673 -1.3854 10.4947>, rot=<0.014 -0.9345 -0.0134 -0.3555>}
 deform_upperarm_armpit_r {parent=deform_upperarm_armpit_base_r, loc=<0 0 0>, rot=<0 0 0 1>}
 upperarm_twist_02_r {parent=upperarm_r, loc=<-13.7941 0 0>, rot=<0 0 0 1>}
 dyn_tie_01 {parent=spine_05, loc=<0.5712 15.1146 -0.1224>, rot=<0.1157 0.6761 0.7164 0.1273>}
 dyn_tie_02 {parent=dyn_tie_01, loc=<7.7687 0 0>, rot=<0 0.0081 0.0022 -1>}
 dyn_tie_03 {parent=dyn_tie_02, loc=<7.0763 0 0>, rot=<-0.0436 0 0 -0.999>}
 deform_pec_aim_base_l {parent=spine_04, loc=<6.2611 9.3309 -1.2237>, rot=<0.0603 -0.4033 0.3318 -0.8507>}
 deform_pec_aim_l {parent=deform_pec_aim_base_l, loc=<0 0 0>, rot=<0 0 0 1>}
 deform_pec_base_l {parent=deform_pec_aim_l, loc=<9.3665 0 0>, rot=<0 0 0 1>}
 deform_pec_l {parent=deform_pec_base_l, loc=<0 0 0>, rot=<0 0 0 1>}
 deform_pec_aim_base_r {parent=spine_04, loc=<6.2607 9.3309 1.2237>, rot=<-0.4033 -0.0603 0.8507 0.3318>}
 deform_pec_aim_r {parent=deform_pec_aim_base_r, loc=<0 0 0>, rot=<0 0 0 1>}
 deform_pec_base_r {parent=deform_pec_aim_r, loc=<-9.3667 -0.0003 -0.0004>, rot=<0 0 0 1>}
 deform_pec_r {parent=deform_pec_base_r, loc=<0 0 0>, rot=<0 0 0 1>}
 thigh_l {parent=pelvis, loc=<-2.5693 -0.0109 -7.7501>, rot=<0.0012 0.0572 0.0155 -0.9982>}
 calf_l {parent=thigh_l, loc=<-39.0484 0 0>, rot=<0.0006 -0.0161 -0.0391 -0.9991>}
 foot_l {parent=calf_l, loc=<-42.561 0.1077 0.4292>, rot=<0.0418 -0.0455 0.0343 -0.9975>}
 ball_l {parent=foot_l, loc=<-6.1599 13.8264 0.6806>, rot=<0.0001 -0.0001 -0.7063 -0.7079>}
 calf_twist_02_l {parent=calf_l, loc=<-18.8064 -0.0967 0.0397>, rot=<0.0237 0.0095 -0.0035 -0.9997>}
 calf_twist_01_l {parent=calf_l, loc=<-35.813 -0.0967 0.0397>, rot=<0.0237 0.0095 -0.0035 -0.9997>}
 thigh_twist_01_l {parent=thigh_l, loc=<-21.6303 0 0>, rot=<0.0013 0 0 -1>}
 dyn_apron_fr_l_base {parent=thigh_l, loc=<0.1717 11.1509 -5.453>, rot=<-0.0217 -0.8398 -0.5377 0.0714>}
 dyn_apron_fr_l_01 {parent=dyn_apron_fr_l_base, loc=<0 0 0>, rot=<-0.0003 -0.0004 -0.0003 -1>}
 dyn_apron_fr_l_02 {parent=dyn_apron_fr_l_01, loc=<13.1597 0 0>, rot=<-0.0118 0.0355 0.0134 -0.9992>}
 dyn_apron_fr_l_03 {parent=dyn_apron_fr_l_02, loc=<14.8181 0 0>, rot=<0.0226 0.0131 0.002 -0.9997>}
 dyn_apron_fr_l_04 {parent=dyn_apron_fr_l_03, loc=<15.8795 0 0>, rot=<-0.0065 -0.0032 0.0005 -1>}
 dyn_apron_fr_l_05 {parent=dyn_apron_fr_l_04, loc=<12.9574 0 0>, rot=<0 0 0 1>}
 dyn_apron_side_l_base {parent=thigh_l, loc=<-0.5499 -0.3953 -10.3565>, rot=<-0.0018 -0.9847 -0.1733 0.0161>}
 dyn_apron_side_l_01 {parent=dyn_apron_side_l_base, loc=<0 0 0>, rot=<0.0005 0.0002 0.0001 -1>}
 dyn_apron_side_l_02 {parent=dyn_apron_side_l_01, loc=<13.1073 0 0>, rot=<-0.0349 0.0134 0.0041 -0.9993>}
 dyn_apron_side_l_03 {parent=dyn_apron_side_l_02, loc=<14.902 0 0>, rot=<0.018 -0.0158 0.0003 -0.9997>}
 dyn_apron_side_l_04 {parent=dyn_apron_side_l_03, loc=<15.8904 0 0>, rot=<-0.008 -0.0111 -0.002 -0.9999>}
 dyn_apron_side_l_05 {parent=dyn_apron_side_l_04, loc=<12.9731 0 0>, rot=<0.0436 0 0 -0.999>}
 thigh_r {parent=pelvis, loc=<-2.5693 -0.0109 7.7501>, rot=<0.0572 -0.0012 0.9982 0.0155>}
 calf_r {parent=thigh_r, loc=<39.0484 0 0>, rot=<0.0006 -0.0161 -0.0391 -0.9991>}
 foot_r {parent=calf_r, loc=<42.561 -0.1077 -0.4292>, rot=<0.0418 -0.0455 0.0343 -0.9975>}
 ball_r {parent=foot_r, loc=<6.1599 -13.8264 -0.6806>, rot=<0.0001 -0.0001 -0.7063 -0.7079>}
 calf_twist_02_r {parent=calf_r, loc=<18.8064 0.0967 -0.0397>, rot=<0.0237 0.0095 -0.0035 -0.9997>}
 calf_twist_01_r {parent=calf_r, loc=<35.813 0.0967 -0.0397>, rot=<0.0237 0.0095 -0.0035 -0.9997>}
 thigh_twist_01_r {parent=thigh_r, loc=<21.6303 0 0>, rot=<0.0013 0 0 -1>}
 dyn_apron_side_r_base {parent=thigh_r, loc=<0.5886 0.1095 10.1777>, rot=<-0.0731 -0.0134 0.0034 -0.9972>}
 dyn_apron_side_r_01 {parent=dyn_apron_side_r_base, loc=<0 0 0>, rot=<0.0016 -0.0014 -0.0012 -1>}
 dyn_apron_side_r_02 {parent=dyn_apron_side_r_01, loc=<13.073 0 0>, rot=<-0.0781 0.0132 -0.0045 -0.9968>}
 dyn_apron_side_r_03 {parent=dyn_apron_side_r_02, loc=<14.8706 0 0>, rot=<-0.0406 -0.0178 0.0071 -0.999>}
 dyn_apron_side_r_04 {parent=dyn_apron_side_r_03, loc=<15.8839 0 0>, rot=<0.0178 -0.013 0.0049 -0.9997>}
 dyn_apron_side_r_05 {parent=dyn_apron_side_r_04, loc=<12.971 0 0>, rot=<0 0 0 1>}
 dyn_apron_fr_r_base {parent=thigh_r, loc=<-0.2308 -11.205 5.434>, rot=<-0.5039 -0.0759 0.0249 -0.8601>}
 dyn_apron_fr_r_01 {parent=dyn_apron_fr_r_base, loc=<0 0 0>, rot=<0.0007 -0.0005 -0.0014 -1>}
 dyn_apron_fr_r_02 {parent=dyn_apron_fr_r_01, loc=<13.2333 0 0>, rot=<0.022 0.0381 -0.0144 -0.9989>}
 dyn_apron_fr_r_03 {parent=dyn_apron_fr_r_02, loc=<14.7944 0 0>, rot=<-0.0583 0.0122 -0.0009 -0.9982>}
 dyn_apron_fr_r_04 {parent=dyn_apron_fr_r_03, loc=<15.8868 0 0>, rot=<-0.0493 -0.0009 0.0049 -0.9988>}
 dyn_apron_fr_r_05 {parent=dyn_apron_fr_r_04, loc=<12.9566 0 0>, rot=<0.0436 0 0 -0.999>}
 dyn_apron_fr_c_base {parent=pelvis, loc=<-1.1942 17.3644 0.0145>, rot=<0.0244 -0.7071 -0.7061 0.03>}
 dyn_apron_fr_c_01 {parent=dyn_apron_fr_c_base, loc=<0 0 0>, rot=<0.0002 0.0005 -0.001 -1>}
 dyn_apron_fr_c_02 {parent=dyn_apron_fr_c_01, loc=<13.2223 0 0>, rot=<-0.0531 0.0547 -0.0027 -0.9971>}
 dyn_apron_fr_c_03 {parent=dyn_apron_fr_c_02, loc=<14.819 0 0>, rot=<0.007 0.0035 -0.0008 -1>}
 dyn_apron_fr_c_04 {parent=dyn_apron_fr_c_03, loc=<15.8707 0 0>, rot=<0.009 -0.0039 0.0001 -1>}
 dyn_apron_fr_c_05 {parent=dyn_apron_fr_c_04, loc=<12.9446 0 0>, rot=<0 0 0 1>}
 dyn_apronTie_01 {parent=pelvis, loc=<-1.3924 -5.6607 -16.9631>, rot=<-0.04 -0.9768 0.1586 -0.1383>}
 dyn_apronTie_02 {parent=dyn_apronTie_01, loc=<3.8938 0 0>, rot=<0 -0.0783 0 -0.9969>}
 dyn_apronTie_03 {parent=dyn_apronTie_02, loc=<8.2581 0 0>, rot=<-0.2578 -0.0226 0.0842 -0.9623>}
]
variables [
 Is On Grind Rail {type=bool, default=False}
 Is Playing Emote {type=bool, default=False}
 L calfBend Alpha {type=float, default=0}
 L thighForward Alpha {type=float, default=0}
 L thighInSide Alpha {type=float, default=0}
 L thighOutSide Alpha {type=float, default=0}
 L thighTwistIn Alpha {type=double, default=0}
 R calfBend Alpha {type=float, default=0}
 R thighForward Alpha {type=float, default=0}
 R thighInSide Alpha {type=float, default=0}
 R thighOutSide Alpha {type=float, default=0}
 R thighTwistIn Alpha {type=double, default=0}
 isDBNO {type=bool, default=False}
]
