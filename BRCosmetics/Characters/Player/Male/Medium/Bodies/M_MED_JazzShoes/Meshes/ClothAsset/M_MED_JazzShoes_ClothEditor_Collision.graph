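# Physics asset that, by its name and ClothAsset folder, supplies collision for cloth on the medium male JazzShoes player body.
# Capsules on the pelvis, spine_01 and spine_02 cover the lower torso, while a wide box on root sits below the origin as a floor.
# Thigh, calf and foot bodies on both sides list no shapes, and the asset has no constraints linking bodies.
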PhysicsAsset /BRCosmetics/Characters/Player/Male/Medium/Bodies/M_MED_JazzShoes/Meshes/ClothAsset/M_MED_JazzShoes_ClothEditor_Collision
bodies {
 calf_l []
 calf_r []
 foot_l []
 foot_r []
 pelvis [{shapeType=Capsule, length=8.6456, radius=14.5388, center=<-2.4887 0.1881 0>}]
 root [{shapeType=Box, extentX=263.5234, extentY=263.5234, extentZ=58.1585, shapeName=root_box, center=<0 0 -29.1195>, rotation=<rot 0 -90 0>}]
 spine_01 [{shapeType=Capsule, length=3.9438, radius=17.5586, center=<-1.7965 0.1884 0>}]
 spine_02 [{shapeType=Capsule, length=3.3264, radius=17.6322, center=<-1.711 -0.1851 0.0001>}]
 thigh_l []
 thigh_r []
}
constraints []
