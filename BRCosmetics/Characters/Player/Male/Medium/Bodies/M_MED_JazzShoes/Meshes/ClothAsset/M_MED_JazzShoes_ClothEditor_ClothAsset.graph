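# Chaos Cloth asset with an embedded Dataflow graph that builds simulated cloth for the male medium JazzShoes player body cosmetic.
# Merges separate render and sim mesh imports, then binds them by proxy deformer and skinning blend.
# Weight maps drive max distance, aerodynamics, anim drive and backstop radius, with backstop using a separately imported accessory mesh.
# A config chain adds material, tethers, clamped root-relative motion, physics-asset collision and solver settings before the single-LOD terminal.
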ChaosClothAsset /BRCosmetics/Characters/Player/Male/Medium/Bodies/M_MED_JazzShoes/Meshes/ClothAsset/M_MED_JazzShoes_ClothEditor_ClothAsset
name=EmbeddedDataflow
graph=&self:EmbeddedDataflow
connections [
 {
  fromNode=LiftDrag WeightMap_1
  fromPinGuid=#1
  toNode=SimulationAerodynamicsConfig_1
  toPinGuid=#2
 }
 {
  fromNode=ProceduralSelection Render
  fromPin=OutputName
  fromPinGuid=#3
  toNode=ProxyDeformer
  toPin="SelectionFilterSets[0].RenderSelection.StringValue"
  toPinGuid=#4
 }
 {
  fromNode=SimulationAerodynamicsConfig_1
  fromPinGuid=#5
  toNode=SimulationGravityConfig_1
  toPinGuid=#6
 }
 {
  fromNode=Backstop Radius WeightMap_1
  fromPin=OutputName.StringValue
  fromPinGuid=#7
  toNode=SimulationBackstopConfig_1
  toPin=BackstopRadius.WeightMap
  toPinGuid=#8
 }
 {
  fromNode=SkeletalMeshImport_Render Mesh_1
  fromPinGuid=#9
  toNode=MergeClothCollections
  toPin="Collections[0]"
  toPinGuid=#10
 }
 {
  fromNode=MergeClothCollections
  fromPinGuid=#11
  toNode=SimAccessoryMeshNode_1
  toPinGuid=#12
 }
 {
  fromNode=LiftDrag WeightMap_1
  fromPin=OutputName.StringValue
  fromPinGuid=#13
  toNode=SimulationAerodynamicsConfig_1
  toPin=Lift.WeightMap
  toPinGuid=#14
 }
 {
  fromNode=SimulationSolverConfig
  fromPinGuid=#15
  toNode=ReferenceBone_1
  toPinGuid=#16
 }
 {
  fromNode=MaxDist WeightMap_1
  fromPin=OutputName.StringValue
  fromPinGuid=#17
  toNode=SimulationMaxDistanceConfig
  toPin=MaxDistance.WeightMap
  toPinGuid=#18
 }
 {
  fromNode=SimulationBendingConfig_3
  fromPinGuid=#19
  toNode=SimulationStretchConfig_3
  toPinGuid=#20
 }
 {
  fromNode=SimulationMaxDistanceConfig
  fromPin=KinematicVertices3D
  fromPinGuid=#21
  toNode=SkinningBlend_1
  toPin=KinematicVertices3D.StringValue
  toPinGuid=#22
 }
 {
  fromNode=SimulationGravityConfig_1
  fromPinGuid=#23
  toNode=SimulationVelocityScaleConfig_1
  toPinGuid=#24
 }
 {
  fromNode=SimulationMassConfig_1
  fromPinGuid=#25
  toNode=SimulationLongRangeAttachmentConfig
  toPinGuid=#26
 }
 {
  fromNode=SimulationCollisionConfig
  fromPinGuid=#27
  toNode=SimulationSolverConfig
  toPinGuid=#28
 }
 {
  fromNode=SetPhysicsAsset - Capsules
  fromPinGuid=#29
  toNode=SimulationCollisionConfig
  toPinGuid=#30
 }
 {
  fromNode=SkeletalMeshImport_Accessory Mesh_1
  fromPinGuid=#31
  toNode=SimAccessoryMeshNode_1
  toPin=SimAccessoryMeshCollection
  toPinGuid=#32
 }
 {
  fromNode=ReferenceBone_1
  fromPinGuid=#33
  toNode=ClothAssetTerminal
  toPin="CollectionLods[0]"
  toPinGuid=#34
 }
 {
  fromNode=SimulationLongRangeAttachmentConfig
  fromPinGuid=#35
  toNode=SimulationBendingConfig_3
  toPinGuid=#36
 }
 {
  fromNode=SimulationStretchConfig_3
  fromPinGuid=#37
  toNode=SimulationDampingConfig_1
  toPinGuid=#38
 }
 {
  fromNode=Backstop Radius WeightMap_1
  fromPinGuid=#39
  toNode=SimulationBackstopConfig_1
  toPinGuid=#40
 }
 {
  fromNode=ProceduralSelection SIM
  fromPin=OutputName
  fromPinGuid=#41
  toNode=ProxyDeformer
  toPin="SelectionFilterSets[0].SimSelection.StringValue"
  toPinGuid=#42
 }
 {
  fromNode=AnimDrive WeightMap_1
  fromPin=OutputName.StringValue
  fromPinGuid=#43
  toNode=SimulationAnimDriveConfig_1
  toPin=AnimDriveStiffness.WeightMap
  toPinGuid=#44
 }
 {
  fromNode=WeightMap_1
  fromPinGuid=#45
  toNode=SimulationMassConfig_1
  toPinGuid=#46
 }
 {
  fromNode=AnimDrive WeightMap_1
  fromPinGuid=#47
  toNode=SimulationAnimDriveConfig_1
  toPinGuid=#48
 }
 {
  fromNode=LiftDrag WeightMap_1
  fromPin=OutputName.StringValue
  fromPinGuid=#13
  toNode=SimulationAerodynamicsConfig_1
  toPin=Drag.WeightMap
  toPinGuid=#49
 }
 {
  fromNode=SimulationMaxDistanceConfig
  fromPinGuid=#50
  toNode=SkinningBlend_1
  toPinGuid=#51
 }
 {
  fromNode=SimulationDampingConfig_1
  fromPinGuid=#52
  toNode=LiftDrag WeightMap_1
  toPinGuid=#53
 }
 {
  fromNode=SimulationVelocityScaleConfig_1
  fromPinGuid=#54
  toNode=AnimDrive WeightMap_1
  toPinGuid=#55
 }
 {
  fromNode=SimulationAnimDriveConfig_1
  fromPinGuid=#56
  toNode=Backstop Radius WeightMap_1
  toPinGuid=#57
 }
 {
  fromNode=MaxDist WeightMap_1
  fromPinGuid=#58
  toNode=SimulationMaxDistanceConfig
  toPinGuid=#59
 }
 {fromNode=ProxyDeformer, fromPinGuid=#60, toNode=MaxDist WeightMap_1, toPinGuid=#61}
 {
  fromNode=SimulationBackstopConfig_1
  fromPinGuid=#62
  toNode=SetPhysicsAsset - Capsules
  toPinGuid=#63
 }
 {
  fromNode=SkeletalMesh_1
  fromPin=SkeletalMesh
  fromPinGuid=#64
  toNode=SkeletalMeshImport_Render Mesh_1
  toPin=SkeletalMesh
  toPinGuid=#65
 }
 {
  fromNode=SimAccessoryMeshNode_1
  fromPinGuid=#66
  toNode=ProceduralSelection Render
  toPinGuid=#67
 }
 {
  fromNode=ProceduralSelection SIM
  fromPinGuid=#68
  toNode=ProxyDeformer
  toPinGuid=#69
 }
 {fromNode=SkinningBlend_1, fromPinGuid=#70, toNode=WeightMap_1, toPinGuid=#71}
 {
  fromNode=ProceduralSelection Render
  fromPinGuid=#72
  toNode=ProceduralSelection SIM
  toPinGuid=#73
 }
 {
  fromNode=SkeletalMeshImport_Sim Mesh_1
  fromPinGuid=#74
  toNode=MergeClothCollections
  toPin="Collections[1]"
  toPinGuid=#75
 }
]
nodes [
 ClothAssetTerminal {type=FChaosClothAssetTerminalNode_v2, guid=#76}
 SimulationMaxDistanceConfig {
  type=FChaosClothAssetSimulationMaxDistanceConfigNode
  guid=#77
  properties {MaxDistance=(bIsAnimatable=True,High=200,WeightMap="MaxDistance")}
 }
 SimulationSolverConfig {
  type=FChaosClothAssetSimulationSolverConfigNode
  guid=#78
  properties {MaxNumIterations=3, NumIterations=1, NumSubstepsImported=(ImportedValue=2)}
 }
 SkeletalMesh_1 {
  type=FGetSkeletalMeshDataflowNode
  guid=#79
  inputPins []
  outputPins [SkeletalMesh {type="TObjectPtr<USkeletalMesh>"}]
  properties {
   PropertyName=SkeletalMesh
   SkeletalMesh=&SkeletalMesh'&/M_MED_JazzShoes_ClothEditor'
  }
 }
 SkeletalMeshImport_Sim Mesh_1 {
  type=FChaosClothAssetSkeletalMeshImportNode_v2
  guid=#80
  properties {
   SkeletalMesh=&SkeletalMesh'&/M_MED_JazzShoes_More_ClothEditor_SIM'
   UVScale=(X=120,Y=120)
   bImportRenderMesh=False
   bImportSimMesh=True
   bImportSingleSection=False
  }
 }
 SkeletalMeshImport_Render Mesh_1 {
  type=FChaosClothAssetSkeletalMeshImportNode_v2
  guid=#81
  properties {SkeletalMesh=None, bImportRenderMesh=True, bImportSimMesh=False, bImportSingleSection=True}
 }
 MergeClothCollections {type=FChaosClothAssetMergeClothCollectionsNode_v2, guid=#82}
 MaxDist WeightMap_1 {type=FChaosClothAssetWeightMapNode, guid=#83, properties {OutputName=(StringValue="MaxDistance")}}
 SimulationLongRangeAttachmentConfig {
  type=FChaosClothAssetSimulationLongRangeAttachmentConfigNode_v2
  guid=#84
  inputPins [
   Collection {type=FManagedArrayCollection}
   FixedEndSet.StringValue {type=FString}
   TetherStiffness.WeightMap {type=FString}
   TetherScale.WeightMap {type=FString}
   "CustomTetherData[0].CustomFixedEndSet.StringValue" {type=FString}
   "CustomTetherData[0].CustomDynamicEndSet.StringValue" {type=FString}
  ]
  outputPins [Collection {type=FManagedArrayCollection}]
  properties {
   CustomTetherData=(())
   FixedEndSet=(StringValue="KinematicVertices3D")
   TetherScale=(bIsAnimatable=True,Low=1,High=1,WeightMap="TetherScale")
   TetherStiffness=(bIsAnimatable=True,Low=1,High=1,WeightMap="TetherStiffness")
   bEnableCustomTetherGeneration=False
   bUseGeodesicTethers=True
   bWarnDuplicateProperty=True
  }
 }
 SimulationBackstopConfig_1 {
  type=FChaosClothAssetSimulationBackstopConfigNode
  guid=#85
  inputPins [
   Collection {type=FManagedArrayCollection}
   BackstopDistance.WeightMap {type=FString}
   BackstopRadius.WeightMap {type=FString}
   BackstopMeshName.StringValue {type=FString}
  ]
  outputPins [Collection {type=FManagedArrayCollection}]
  properties {
   BackstopDistance=(bIsAnimatable=True,WeightMap="BackstopDistance")
   BackstopMeshName=(StringValue="AccessoryMesh")
   BackstopRadius=(bIsAnimatable=True,Low=20,High=60,WeightMap="BackstopRadius")
   bWarnDuplicateProperty=True
  }
 }
 SimulationAnimDriveConfig_1 {
  type=FChaosClothAssetSimulationAnimDriveConfigNode
  guid=#86
  inputPins [
   Collection {type=FManagedArrayCollection}
   AnimDriveStiffness.WeightMap {type=FString}
   AnimDriveDamping.WeightMap {type=FString}
  ]
  outputPins [Collection {type=FManagedArrayCollection}]
  properties {
   AnimDriveDamping=(bIsAnimatable=True,High=1,WeightMap="AnimDriveDamping")
   AnimDriveStiffness=(bIsAnimatable=True,Low=0.2,High=1,WeightMap="AnimDriveStiffness")
   bWarnDuplicateProperty=True
  }
 }
 AnimDrive WeightMap_1 {type=FChaosClothAssetWeightMapNode, guid=#87, properties {OutputName=(StringValue="AnimDrive")}}
 SimulationAerodynamicsConfig_1 {
  type=FChaosClothAssetSimulationAerodynamicsConfigNode
  guid=#88
  properties {
   Drag=(bIsAnimatable=True,Low=0.25,High=0.6,WeightMap="Drag")
   Lift=(bIsAnimatable=True,Low=0.2,High=0.3,WeightMap="Lift")
   TurbulenceRatio=1
   WindVelocity=(X=0,Y=0,Z=0)
   WindVelocitySpace=ReferenceBoneSpace
  }
 }
 LiftDrag WeightMap_1 {type=FChaosClothAssetWeightMapNode, guid=#89, properties {OutputName=(StringValue="LiftDrag")}}
 SimulationBendingConfig_3 {
  type=FChaosClothAssetSimulationBendingConfigNode
  guid=#90
  inputPins [
   Collection {type=FManagedArrayCollection}
   FlatnessRatio.WeightMap {type=FString}
   RestAngle.WeightMap {type=FString}
   BendingStiffness.WeightMap {type=FString}
   BendingStiffnessWarp.WeightMap {type=FString}
   BendingStiffnessWeft.WeightMap {type=FString}
   BendingStiffnessBias.WeightMap {type=FString}
   BendingDamping.WeightMap {type=FString}
   BendingAnisoDamping.WeightMap {type=FString}
   AnisoBucklingRatio.WeightMap {type=FString}
   BucklingStiffness.WeightMap {type=FString}
   BucklingStiffnessWarp.WeightMap {type=FString}
   BucklingStiffnessWeft.WeightMap {type=FString}
   BucklingStiffnessBias.WeightMap {type=FString}
   BucklingRatioWeighted.WeightMap {type=FString}
  ]
  properties {
   AnisoBucklingRatio=(bIsAnimatable=True,Low=0.5,High=0.5,WeightMap="BucklingRatio")
   BendingAnisoDamping=(bIsAnimatable=True,Low=0.5,High=1,WeightMap="BendingAnisoDamping")
   BendingStiffnessBias=(bIsAnimatable=True,Low=100,High=100,WeightMap="BendingStiffnessBias")
   BendingStiffnessWarp=(bIsAnimatable=True,Low=100,High=100,WeightMap="BendingStiffnessWarp")
   BendingStiffnessWeft=(bIsAnimatable=True,Low=100,High=100,WeightMap="BendingStiffnessWeft")
   BucklingStiffnessBias=(bIsAnimatable=True,Low=2,High=50,WeightMap="BucklingStiffnessBias")
   BucklingStiffnessWarp=(bIsAnimatable=True,Low=2,High=50,WeightMap="BucklingStiffnessWarp")
   BucklingStiffnessWeft=(bIsAnimatable=True,Low=2,High=50,WeightMap="BucklingStiffnessWeft")
  }
 }
 SimulationStretchConfig_3 {
  type=FChaosClothAssetSimulationStretchConfigNode
  guid=#91
  properties {
   StretchAnisoDamping=(bIsAnimatable=True,Low=1,High=1,WeightMap="StretchAnisoDamping")
   StretchStiffnessBias=(bIsAnimatable=True,Low=0.5,High=0.8,WeightMap="StretchStiffnessBias")
   StretchStiffnessWarp=(bIsAnimatable=True,Low=10,High=50,WeightMap="StretchStiffnessWarp")
   StretchStiffnessWeft=(bIsAnimatable=True,Low=0.5,High=5,WeightMap="StretchStiffnessWeft")
  }
 }
 SimulationMassConfig_1 {
  type=FChaosClothAssetSimulationMassConfigNode
  guid=#92
  properties {DensityWeighted=(Low=0.7,High=0.7,WeightMap="Density")}
 }
 SimulationDampingConfig_1 {
  type=FChaosClothAssetSimulationDampingConfigNode
  guid=#93
  properties {
   DampingCoefficientWeighted=(WeightMap="DampingCoefficient")
   LocalDampingAngularCoefficient=0.05
   LocalDampingLinearCoefficient=0.05
   LocalDampingSpace=ReferenceBoneSpace
  }
 }
 SimulationGravityConfig_1 {
  type=FChaosClothAssetSimulationGravityConfigNode
  guid=#94
  properties {GravityOverrideImported=(ImportedValue=(X=0,Y=0,Z=-980.665))}
 }
 SimulationVelocityScaleConfig_1 {
  type=FChaosClothAssetSimulationVelocityScaleConfigNode
  guid=#95
  inputPins [Collection {type=FManagedArrayCollection}]
  outputPins [Collection {type=FManagedArrayCollection}]
  properties {
   AngularVelocityScale=0.1
   FictitiousAngularScale=0.1
   FictitiousForcesModel=FullScale
   LinearVelocityScale=(X=0.15,Y=0.15,Z=0.3)
   MaxAngularAcceleration=3
   MaxAngularVelocity=1
   MaxLinearAcceleration=(X=1500,Y=1500,Z=2000)
   MaxLinearVelocity=(X=500,Y=500,Z=600)
   MaxVelocityScale=1
   VelocityScaleSpace=ReferenceBoneSpace
   bEnableAngularAccelerationClamping=True
   bEnableAngularVelocityClamping=True
   bEnableLinearAccelerationClamping=True
   bEnableLinearVelocityClamping=True
   bWarnDuplicateProperty=True
  }
 }
 SimulationCollisionConfig {
  type=FChaosClothAssetSimulationCollisionConfigNode
  guid=#96
  properties {
   ClothCollisionThicknessImported=(bIsAnimatable=True,WeightMap="ClothCollisionThickness")
   CollisionThicknessImported=(ImportedValue=0.5)
   FrictionCoefficientWeighted=(bIsAnimatable=True,Low=0.5,High=0.5,WeightMap="FrictionCoefficient")
   InnerCollisionThickness=(Low=1,WeightMap="InnerCollisionThickness")
   bUseCCD=False
  }
 }
 SetPhysicsAsset - Capsules {
  type=FChaosClothAssetSetPhysicsAssetNode
  guid=#97
  properties {PhysicsAsset=/Script/Engine.PhysicsAsset'&/M_MED_JazzShoes_ClothEditor_Collision'}
 }
 WeightMap_1 {
  type=FChaosClothAssetWeightMapNode
  guid=#98
  properties {MapOverrideType=Add, MeshTarget=Render, OutputName=()}
 }
 ProxyDeformer {
  type=FChaosClothAssetProxyDeformerNode_v3
  guid=#99
  properties {InfluenceRadius=5, SelectionFilterSets=(()), bPreserveRenderTangents=True}
 }
 ReferenceBone_1 {
  type=FChaosClothAssetReferenceBoneNode
  guid=#100
  inputPins [Collection {type=FManagedArrayCollection}]
  outputPins [Collection {type=FManagedArrayCollection}]
  properties {CalculateDefaultReferenceBone=(), ReferenceBone=(Name="root")}
 }
 SkinningBlend_1 {
  type=FChaosClothAssetSkinningBlendNode
  guid=#101
  inputPins [
   Collection {type=FManagedArrayCollection}
   KinematicVertices3D.StringValue {type=FString}
  ]
  outputPins [
   Collection {type=FManagedArrayCollection}
   SkinningBlendName {type=FString}
  ]
  properties {KinematicVertices3D=(), SkinningBlendName=RenderDeformerSkinningBlend, bUseSmoothTransition=True}
 }
 SimAccessoryMeshNode_1 {
  type=FChaosClothAssetSimAccessoryMeshNode
  guid=#102
  inputPins [
   Collection {type=FManagedArrayCollection}
   SimAccessoryMeshCollection {type=FManagedArrayCollection}
  ]
  outputPins [
   Collection {type=FManagedArrayCollection}
   AccessoryMeshName {type=FString}
  ]
  properties {AccessoryMeshName=AccessoryMesh, bUseSimImportVertexID=True}
 }
 SkeletalMeshImport_Accessory Mesh_1 {
  type=FChaosClothAssetSkeletalMeshImportNode_v2
  guid=#103
  properties {
   SkeletalMesh=&SkeletalMesh'&/M_MED_JazzShoes_More_ClothEditor_BACKSTOP'
   UVScale=(X=120,Y=120)
   bImportRenderMesh=False
   bImportSimMesh=True
   bImportSingleSection=False
  }
 }
 Backstop Radius WeightMap_1 {type=FChaosClothAssetWeightMapNode, guid=#104, properties {OutputName=(StringValue="BackstopRadius")}}
 ProceduralSelection Render {
  type=FChaosClothAssetProceduralSelectionNode
  guid=#105
  inputPins [
   Collection {type=FManagedArrayCollection}
   ConversionInputName.StringValue {type=FString}
  ]
  outputPins [
   Collection {type=FManagedArrayCollection}
   OutputName {type=FString}
  ]
  properties {ConversionInputName=(), Group=(Name="RenderVertices"), OutputName=RenderSelectionProcedural, SelectionType=SelectAll}
 }
 ProceduralSelection SIM {
  type=FChaosClothAssetProceduralSelectionNode
  guid=#106
  inputPins [
   Collection {type=FManagedArrayCollection}
   ConversionInputName.StringValue {type=FString}
  ]
  outputPins [
   Collection {type=FManagedArrayCollection}
   OutputName {type=FString}
  ]
  properties {ConversionInputName=(), Group=(Name="SimVertices3D"), OutputName=SimSelectionProcedural, SelectionType=SelectAll}
 }
]
